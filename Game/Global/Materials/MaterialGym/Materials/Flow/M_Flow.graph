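# Character material in MaterialGym that adds a flowing emissive effect, driven by oceanbreeze flow and mask textures, to the standard character shading.
# It runs the character PreFX and CharacterSkinFX functions, then darkens base color inside the SkinFX mask.
# MF_Flow output is added to the heavily dimmed skin emissive.
# The result passes through character PostFX and cinematic PostFX, with shading model chosen by material expression.
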Material /Game/Global/Materials/MaterialGym/Materials/Flow/M_Flow
graph {
 properties {
  bAutomaticallySetUsageInEditor=false
  bUseMaterialAttributes=true
  bUsedWithClothing=true
  bUsedWithMeshDeformer=true
  bUsedWithMorphTargets=true
  bUsedWithNanite=true
  bUsedWithSkeletalMesh=true
  pixelDepthOffsetMode=PDOM_Legacy
  refractionMethod=RM_IndexOfRefraction
  shadingModel=MSM_FromMaterialExpression
  subsurfaceProfile=&SS_HeroSkin_02
  usedShadingModels=Default Lit | Subsurface | Clear Coat | Subsurface Profile | Cloth | Eye
 }
 nodes [
  n1 = MaterialFunctionCall() {materialFunction=&BaseFN_CharacterFunction_PreFX}
  n2 = MaterialFunctionCall(
   PreskinnedPosition=@n1.Pre skinned Local Position
   Gmap=@n1.Pre skinned Local Position
   MaterialAttributes=@n1.Pre skinned Local Position
   AO=@n1.Pre skinned Local Position
  ) {materialFunction=&CharacterSkinFX}
  n3 = GetMaterialAttributes(None=@n2.Result) {attributeGetTypes [#1, #2]}
  n4 = Multiply(A=@n3.Emissive Color) {constB=0.1}
  n5 = MaterialFunctionCall!(
   FlowID
   FlowGradient=TextureSample() {samplerType=SAMPLERTYPE_Grayscale, texture=/Game/Global/Materials/MaterialGym/Materials/Flow/Textures/T_oceanbreeze_flowmask}.R
   FlowMask=TextureSample() {samplerType=SAMPLERTYPE_LinearColor, texture=/Game/Global/Materials/MaterialGym/Materials/Flow/Textures/T_oceanbreeze_mask}.R
  ) {materialFunction=/Game/Global/Materials/MaterialGym/Materials/Flow/MaterialFunctions/MF_Flow}
  n6 = Multiply(A=@n3.Base Color) {constB=0.25}
  n7 = ComponentMask(None=@n2.SkinFX_Mask) {r=true}
  n8 = MaterialFunctionCall(
   AO=@n1.AO
   Distance Blend=@n1.AO
   Fresnel=@n1.AO
   Gmap=@n1.AO
   Pre Skinned Local Position=@n1.AO
   Material=SetMaterialAttributes!(
    MaterialAttributes=@n3.MaterialAttributes
    Emissive Color=Add(A=@n4, B=@n5.Result)
    Base Color=LinearInterpolate(A=@n3.Base Color, B=@n6, Alpha=@n7)
   ) {attributeSetTypes [#1, #2]}
  ) {materialFunction=&BaseFN_CharacterFunction_PostFX}
  MaterialFunctionCall(In=@n8.Result) {materialFunction=&CINE_MF_CharPostFX}
 ]
}
